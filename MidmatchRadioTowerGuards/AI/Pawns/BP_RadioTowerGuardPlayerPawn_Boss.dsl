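; Blueprint for the boss player pawn of the midmatch Radio Tower Guards AI, with logic to resolve a playlist squad-size tag.
; GetPlaylistSquadSizeTag compares each entry in its local playlist tag list against the Athena game state's playlist context tags and returns the first match.
; It falls back to Athena.Playlist.Duo when the game state is invalid or nothing matches.
; The construction script only calls the parent.
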(fn GetPlaylistSquadSizeTag
  (bind _returnvalue (Fortnite|Athena|GameState|GetGameStateAthena))
  ;[0.01563 0.08401 0.1094] Returns the playlist tag that determines the squad size.
  (Utilities|FlowControl|Sequence
    (:then_0
      (Utilities|IsValid _returnvalue
        (:Is%20Valid
          (for _array_element (LocalVariables|GetPlaylistTags)
            (if (GameplayTags|TagContainer|CompareTagContainertoOtherTags _array_element (Class|FortGameStateAthena|GetAthenaPlaylistContextTags _returnvalue))
              (return _array_element))))))
    (:then_1
      (return "(TagName=\"Athena.Playlist.Duo\")"))))

(fn ConstructionScript
  (|CallParentFunction))
